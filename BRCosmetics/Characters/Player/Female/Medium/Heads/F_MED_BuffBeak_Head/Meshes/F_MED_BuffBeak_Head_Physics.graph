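# Physics asset for the F_MED_BuffBeak_Head cosmetic, a female medium player head with left and right fin bones.
# Three sphere bodies cover the head and both fins, with the fin spheres larger than the head's and slightly offset from their bones.
# Each fin is constrained to the head with identical narrow swing and twist limits, allowing only slight movement relative to it.
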PhysicsAsset /BRCosmetics/Characters/Player/Female/Medium/Heads/F_MED_BuffBeak_Head/Meshes/F_MED_BuffBeak_Head_Physics
bodies {
 fin_l [{shapeType=Sphere, radius=3, shapeName=fin_l_sphere, center=<0 -0.8285 0>}]
 fin_r [{shapeType=Sphere, radius=3, shapeName=fin_r_sphere, center=<0 -0.8285 0>}]
 head [{shapeType=Sphere, radius=1, shapeName=head_sphere}]
}
constraints [
 {bone1Name=fin_l, bone2Name=head, swing1LimitDegrees=12, swing2LimitDegrees=3, twistLimitDegrees=12}
 {bone1Name=fin_r, bone2Name=head, swing1LimitDegrees=12, swing2LimitDegrees=3, twistLimitDegrees=12}
]
