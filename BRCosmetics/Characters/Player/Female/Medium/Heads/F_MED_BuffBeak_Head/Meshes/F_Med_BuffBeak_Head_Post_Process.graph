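# Post-process AnimBlueprint for the female BuffBeak head, adding secondary physics and rig-driven adjustments to the incoming pose.
# The pose runs through a RigidBody node simulated relative to spine_05 with gravity override, then the head's Control Rig.
# The event graph uses random delays and a gate to compute timed Perlin-style eye-twitch rotations.
# The lower-left eyelid bone node reading that rotation is not linked into the output pose.
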AnimBlueprint /BRCosmetics/Characters/Player/Female/Medium/Heads/F_MED_BuffBeak_Head/Meshes/F_Med_BuffBeak_Head_Post_Process
target_skeleton=/Game/Characters/Player/Common/Fortnite_Base_Head/Fortnite_Base_Head_Export_Skeleton
graphs [
 AnimGraph {nodes [
   "Transform (Modify) Bone - Bone: L_eye_lid_lower_mid"!(ComponentPose, bAlphaBoolEnabled:bool=True, Translation="0,0,0", Rotation=Get Rotation Out() -> (Rotation Out:Rotator), Scale="1,1,1") -> (Pose) {
    alphaInputType=Bool
    rotationMode=BMM_Additive
    rotationSpace=BCS_BoneSpace
    alphaBoolBlend {blendInTime=0.2, blendOption=Linear, blendOutTime=0.2}
    boneToModify {boneName=L_eye_lid_lower_mid}
   }
   n1 = RigidBody(ComponentPose=Local To Component(LocalPose=Input Pose())) {
    bClampLinearTranslationLimitToRefPose=true
    bOverrideWorldGravity=true
    overridePhysicsAsset=&/F_MED_BuffBeak_Head_Physics
    simulationSpace=BaseBoneSpace
    alphaBoolBlend {blendOption=Linear}
    baseBoneRef {boneName=spine_05}
    componentLinearAccScale=<0.8 0.8 0.8>
    componentLinearVelScale=<0.8 0.8 0.8>
    overrideWorldGravity {z=-980}
   }
   n2 = Control Rig(Source=Component To Local(ComponentPose=@n1)) {
    alphaBoolBlend {blendOption=Linear}
    controlRigAssetReference {blueprintRigClass=&/F_MED_BuffBeak_Head_CtrlRig.F_MED_BuffBeak_Head_CtrlRig_C}
   }
   Output Pose(Result=@n2) {layerGroup=None}
  ]}
 EventGraph {nodes [
   Event BlueprintUpdateAnimation()
   > n1 = Sequence()
   > Gate!(Enter:Exec, Open:Exec=@n5.then, Close:Exec=@n3.then, Toggle:Exec, bStartClosed:bool=true) -> (Exit:Exec) {is=title K2Node_MacroInstance}
   > n2 = RandomRotatorFunction!(
    execute
    GameTimeInSeconds:double=GetGameTimeInSeconds()
    PerlinScale:double=1
    Frequency:Rotator="0,0,30"
    Amplitude:Rotator="0,0,8"
    Offset:Rotator="0,0,-5"
    TimeOffset:double=0
   ) -> (then, RandomRotator:Rotator)
   > Set Rotation Out(Rotation Out:Rotator=@n2) -> (Output_Get:Rotator)
   > Set bIsEyeTwitching(bIsEyeTwitching:bool=true) -> (Output_Get:bool)
   > Delay(Duration=0.65)
   > n3 = Set bIsEyeTwitching(bIsEyeTwitching=false)
   n4 = Branch(execute=@n1.then_1, Condition=Get bIsEyeTwitching() -> (bIsEyeTwitching:bool))
   n5 = Delay(execute=@n4.else, Duration=RandomFloatInRange(Min=1, Max=3.5))
   TryGetPawnOwner()
   Get Rotation Out()
  ]}
]
variables [
 bIsEyeTwitching {type=boolean}
 Rotation Out {type=Rotator}
]
